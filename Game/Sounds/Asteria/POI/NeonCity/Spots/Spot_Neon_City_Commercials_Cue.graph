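# Sound cue for Neon City ambient spot audio, playing randomized commercial and "newsgrind" voice lines under the ambient sound class with its own attenuation.
# A concatenator plays one clip after a short random delay, then loops further clips with longer random gaps.
# Both branches mirror the same pools, repeating the commercial pool in the random selector so news lines are picked about one time in ten.
# Modulators are neutral, and a concurrency override limits simultaneous instances.
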SoundCue /Game/Sounds/Asteria/POI/NeonCity/Spots/Spot_Neon_City_Commercials_Cue
root=Concatenator_0
nodes [
 {id=Concatenator_0, children [Delay_1,Looping_0], properties {inputVolume [1,1]}}
 {id=Delay_0, children [Modulator_0], properties {delayMax=12, delayMin=6}}
 {id=Delay_1, children [Modulator_1], properties {delayMax=2}}
 {id=Looping_0, children [Delay_0]}
 {id=Modulator_0, children [Random_2], properties {pitchMax=1, pitchMin=1, volumeMax=1, volumeMin=1}}
 {id=Modulator_1, children [Random_4], properties {pitchMax=1, pitchMin=1, volumeMax=1, volumeMin=1}}
 {
  id=Random_0
  children [
   WavePlayer_0
   WavePlayer_1
   WavePlayer_2
   WavePlayer_3
   WavePlayer_4
   WavePlayer_5
   WavePlayer_9
   WavePlayer_13
   WavePlayer_14
   WavePlayer_15
   WavePlayer_16
   WavePlayer_37
   WavePlayer_6
   WavePlayer_7
   WavePlayer_8
   WavePlayer_10
   WavePlayer_11
   WavePlayer_12
   WavePlayer_35
   WavePlayer_36
   WavePlayer_42
  ]
  properties {bShouldExcludeFromBranchCulling=true, weights [1,1,1,1,1,1,1,1,1,1,1,1,1,1,1,1,1,1,1,1,1]}
 }
 {
  id=Random_1
  children [
   WavePlayer_38
   WavePlayer_39
   WavePlayer_40
   WavePlayer_41
   WavePlayer_17
   WavePlayer_18
   WavePlayer_19
   WavePlayer_20
   WavePlayer_21
   WavePlayer_22
  ]
  properties {bShouldExcludeFromBranchCulling=true, weights [1,1,1,1,1,1,1,1,1,1]}
 }
 {
  id=Random_2
  children [Random_0,Random_0,Random_0,Random_0,Random_0,Random_0,Random_0,Random_0,Random_0,Random_1]
  properties {bShouldExcludeFromBranchCulling=true, weights [1,1,1,1,1,1,1,1,1,1]}
 }
 {
  id=Random_3
  children [
   WavePlayer_23
   WavePlayer_24
   WavePlayer_25
   WavePlayer_26
   WavePlayer_27
   WavePlayer_28
   WavePlayer_32
   WavePlayer_44
   WavePlayer_45
   WavePlayer_46
   WavePlayer_47
   WavePlayer_50
   WavePlayer_29
   WavePlayer_30
   WavePlayer_31
   WavePlayer_33
   WavePlayer_34
   WavePlayer_43
   WavePlayer_48
   WavePlayer_49
   WavePlayer_55
  ]
  properties {bShouldExcludeFromBranchCulling=true, weights [1,1,1,1,1,1,1,1,1,1,1,1,1,1,1,1,1,1,1,1,1]}
 }
 {
  id=Random_4
  children [Random_3,Random_3,Random_3,Random_3,Random_3,Random_3,Random_3,Random_3,Random_3,Random_5]
  properties {bShouldExcludeFromBranchCulling=true, weights [1,1,1,1,1,1,1,1,1,1]}
 }
 {
  id=Random_5
  children [
   WavePlayer_51
   WavePlayer_52
   WavePlayer_53
   WavePlayer_54
   WavePlayer_56
   WavePlayer_57
   WavePlayer_58
   WavePlayer_59
   WavePlayer_60
   WavePlayer_61
  ]
  properties {bShouldExcludeFromBranchCulling=true, weights [1,1,1,1,1,1,1,1,1,1]}
 }
 {id=WavePlayer_0, properties {soundWaveAssetPtr=&/comm_vo_aoike_031}}
 {id=WavePlayer_1, properties {soundWaveAssetPtr=&/comm_vo_aoike_033}}
 {id=WavePlayer_10, properties {soundWaveAssetPtr=&/comm_vo_fujita_009}}
 {id=WavePlayer_11, properties {soundWaveAssetPtr=&/comm_vo_fujita_010}}
 {id=WavePlayer_12, properties {soundWaveAssetPtr=&/comm_vo_fujita_011}}
 {id=WavePlayer_13, properties {soundWaveAssetPtr=&/comm_vo_luna_012}}
 {id=WavePlayer_14, properties {soundWaveAssetPtr=&/comm_vo_luna_017}}
 {id=WavePlayer_15, properties {soundWaveAssetPtr=&/comm_vo_luna_018}}
 {id=WavePlayer_16, properties {soundWaveAssetPtr=&/comm_vo_luna_019}}
 {id=WavePlayer_17, properties {soundWaveAssetPtr=&/newsgrind_vo_aoike_018}}
 {id=WavePlayer_18, properties {soundWaveAssetPtr=&/newsgrind_vo_space_001}}
 {id=WavePlayer_19, properties {soundWaveAssetPtr=&/newsgrind_vo_space_002}}
 {id=WavePlayer_2, properties {soundWaveAssetPtr=&/comm_vo_aoike_036}}
 {id=WavePlayer_20, properties {soundWaveAssetPtr=&/newsgrind_vo_space_003}}
 {id=WavePlayer_21, properties {soundWaveAssetPtr=&/newsgrind_vo_space_004}}
 {id=WavePlayer_22, properties {soundWaveAssetPtr=&/newsgrind_vo_space_018}}
 {id=WavePlayer_23, properties {soundWaveAssetPtr=&comm_vo_aoike_031}}
 {id=WavePlayer_24, properties {soundWaveAssetPtr=&comm_vo_aoike_033}}
 {id=WavePlayer_25, properties {soundWaveAssetPtr=&comm_vo_aoike_036}}
 {id=WavePlayer_26, properties {soundWaveAssetPtr=&/comm_vo_aoike_038}}
 {id=WavePlayer_27, properties {soundWaveAssetPtr=&/comm_vo_aoike_040}}
 {id=WavePlayer_28, properties {soundWaveAssetPtr=&/comm_vo_luna_001}}
 {id=WavePlayer_29, properties {soundWaveAssetPtr=&/comm_vo_fujita_003}}
 {id=WavePlayer_3, properties {soundWaveAssetPtr=&comm_vo_aoike_038}}
 {id=WavePlayer_30, properties {soundWaveAssetPtr=&/comm_vo_fujita_004}}
 {id=WavePlayer_31, properties {soundWaveAssetPtr=&/comm_vo_fujita_006}}
 {id=WavePlayer_32, properties {soundWaveAssetPtr=&/comm_vo_luna_008}}
 {id=WavePlayer_33, properties {soundWaveAssetPtr=&comm_vo_fujita_009}}
 {id=WavePlayer_34, properties {soundWaveAssetPtr=&comm_vo_fujita_010}}
 {id=WavePlayer_35, properties {soundWaveAssetPtr=&/comm_vo_fujita_022}}
 {id=WavePlayer_36, properties {soundWaveAssetPtr=&/comm_vo_fujita_023}}
 {id=WavePlayer_37, properties {soundWaveAssetPtr=&/comm_vo_luna_025}}
 {id=WavePlayer_38, properties {soundWaveAssetPtr=&/newsgrind_vo_aoike_001}}
 {id=WavePlayer_39, properties {soundWaveAssetPtr=&/newsgrind_vo_aoike_002}}
 {id=WavePlayer_4, properties {soundWaveAssetPtr=&comm_vo_aoike_040}}
 {id=WavePlayer_40, properties {soundWaveAssetPtr=&/newsgrind_vo_aoike_003}}
 {id=WavePlayer_41, properties {soundWaveAssetPtr=&/newsgrind_vo_aoike_004}}
 {id=WavePlayer_42, properties {soundWaveAssetPtr=&/comm_vo_fujita_040}}
 {id=WavePlayer_43, properties {soundWaveAssetPtr=&comm_vo_fujita_011}}
 {id=WavePlayer_44, properties {soundWaveAssetPtr=&comm_vo_luna_012}}
 {id=WavePlayer_45, properties {soundWaveAssetPtr=&comm_vo_luna_017}}
 {id=WavePlayer_46, properties {soundWaveAssetPtr=&comm_vo_luna_018}}
 {id=WavePlayer_47, properties {soundWaveAssetPtr=&comm_vo_luna_019}}
 {id=WavePlayer_48, properties {soundWaveAssetPtr=&comm_vo_fujita_022}}
 {id=WavePlayer_49, properties {soundWaveAssetPtr=&comm_vo_fujita_023}}
 {id=WavePlayer_5, properties {soundWaveAssetPtr=&comm_vo_luna_001}}
 {id=WavePlayer_50, properties {soundWaveAssetPtr=&comm_vo_luna_025}}
 {id=WavePlayer_51, properties {soundWaveAssetPtr=&newsgrind_vo_aoike_001}}
 {id=WavePlayer_52, properties {soundWaveAssetPtr=&newsgrind_vo_aoike_002}}
 {id=WavePlayer_53, properties {soundWaveAssetPtr=&newsgrind_vo_aoike_003}}
 {id=WavePlayer_54, properties {soundWaveAssetPtr=&newsgrind_vo_aoike_004}}
 {id=WavePlayer_55, properties {soundWaveAssetPtr=&comm_vo_fujita_040}}
 {id=WavePlayer_56, properties {soundWaveAssetPtr=&newsgrind_vo_aoike_018}}
 {id=WavePlayer_57, properties {soundWaveAssetPtr=&newsgrind_vo_space_001}}
 {id=WavePlayer_58, properties {soundWaveAssetPtr=&newsgrind_vo_space_002}}
 {id=WavePlayer_59, properties {soundWaveAssetPtr=&newsgrind_vo_space_003}}
 {id=WavePlayer_6, properties {soundWaveAssetPtr=&comm_vo_fujita_003}}
 {id=WavePlayer_60, properties {soundWaveAssetPtr=&newsgrind_vo_space_004}}
 {id=WavePlayer_61, properties {soundWaveAssetPtr=&newsgrind_vo_space_018}}
 {id=WavePlayer_7, properties {soundWaveAssetPtr=&comm_vo_fujita_004}}
 {id=WavePlayer_8, properties {soundWaveAssetPtr=&comm_vo_fujita_006}}
 {id=WavePlayer_9, properties {soundWaveAssetPtr=&comm_vo_luna_008}}
]
properties {
 attenuationSettings=&/Spot_Neon_City_Commercial_Att
 bOverrideConcurrency=true
 maxDistance=11500
 soundClassObject=&Fort_Ambient_Athena
 volumeMultiplier=0.626
 concurrencyOverrides {maxCount=2, retriggerTime=6}
}
