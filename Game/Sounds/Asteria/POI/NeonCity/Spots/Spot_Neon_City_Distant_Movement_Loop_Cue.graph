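# Ambient spot sound cue for the Neon City POI that loops randomized arcade one-shots, despite its "Distant Movement" name.
# A Concatenator plays one random sound after a short delay, then a Looping branch repeats random picks separated by longer random delays.
# Both branches draw equally from the same 18 arcade waves with random pitch and volume.
# It uses commercial-spot attenuation, an ambient sound class and overridden concurrency limits.
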SoundCue /Game/Sounds/Asteria/POI/NeonCity/Spots/Spot_Neon_City_Distant_Movement_Loop_Cue
root=Concatenator_0
nodes [
 {id=Concatenator_0, children [Delay_1,Looping_0], properties {inputVolume [1,1]}}
 {id=Delay_0, children [Modulator_0], properties {delayMax=9, delayMin=3}}
 {id=Delay_1, children [Modulator_1], properties {delayMax=1}}
 {id=Looping_0, children [Delay_0]}
 {id=Modulator_0, children [Random_1], properties {pitchMax=1.2, pitchMin=0.85, volumeMax=1, volumeMin=0.8}}
 {id=Modulator_1, children [Random_0], properties {pitchMax=1.2, pitchMin=0.85, volumeMax=1, volumeMin=0.8}}
 {
  id=Random_0
  children [
   WavePlayer_0
   WavePlayer_1
   WavePlayer_2
   WavePlayer_3
   WavePlayer_4
   WavePlayer_5
   WavePlayer_24
   WavePlayer_25
   WavePlayer_26
   WavePlayer_27
   WavePlayer_28
   WavePlayer_29
   WavePlayer_30
   WavePlayer_31
   WavePlayer_32
   WavePlayer_33
   WavePlayer_34
   WavePlayer_35
  ]
  properties {bShouldExcludeFromBranchCulling=true, weights [1,1,1,1,1,1,1,1,1,1,1,1,1,1,1,1,1,1]}
 }
 {
  id=Random_1
  children [
   WavePlayer_6
   WavePlayer_7
   WavePlayer_8
   WavePlayer_9
   WavePlayer_10
   WavePlayer_11
   WavePlayer_12
   WavePlayer_13
   WavePlayer_14
   WavePlayer_15
   WavePlayer_16
   WavePlayer_17
   WavePlayer_18
   WavePlayer_19
   WavePlayer_20
   WavePlayer_21
   WavePlayer_22
   WavePlayer_23
  ]
  properties {bShouldExcludeFromBranchCulling=true, weights [1,1,1,1,1,1,1,1,1,1,1,1,1,1,1,1,1,1]}
 }
 {id=WavePlayer_0, properties {soundWaveAssetPtr=&/Spot_Neon_City_Arcade_Sounds_01}}
 {id=WavePlayer_1, properties {soundWaveAssetPtr=&/Spot_Neon_City_Arcade_Sounds_02}}
 {id=WavePlayer_10, properties {soundWaveAssetPtr=&/Spot_Neon_City_Arcade_Sounds_05}}
 {id=WavePlayer_11, properties {soundWaveAssetPtr=&/Spot_Neon_City_Arcade_Sounds_06}}
 {id=WavePlayer_12, properties {soundWaveAssetPtr=&/Spot_Neon_City_Arcade_Sounds_07}}
 {id=WavePlayer_13, properties {soundWaveAssetPtr=&/Spot_Neon_City_Arcade_Sounds_08}}
 {id=WavePlayer_14, properties {soundWaveAssetPtr=&/Spot_Neon_City_Arcade_Sounds_09}}
 {id=WavePlayer_15, properties {soundWaveAssetPtr=&/Spot_Neon_City_Arcade_Sounds_10}}
 {id=WavePlayer_16, properties {soundWaveAssetPtr=&/Spot_Neon_City_Arcade_Sounds_11}}
 {id=WavePlayer_17, properties {soundWaveAssetPtr=&/Spot_Neon_City_Arcade_Sounds_12}}
 {id=WavePlayer_18, properties {soundWaveAssetPtr=&/Spot_Neon_City_Arcade_Sounds_13}}
 {id=WavePlayer_19, properties {soundWaveAssetPtr=&/Spot_Neon_City_Arcade_Sounds_14}}
 {id=WavePlayer_2, properties {soundWaveAssetPtr=&/Spot_Neon_City_Arcade_Sounds_03}}
 {id=WavePlayer_20, properties {soundWaveAssetPtr=&/Spot_Neon_City_Arcade_Sounds_15}}
 {id=WavePlayer_21, properties {soundWaveAssetPtr=&/Spot_Neon_City_Arcade_Sounds_16}}
 {id=WavePlayer_22, properties {soundWaveAssetPtr=&/Spot_Neon_City_Arcade_Sounds_17}}
 {id=WavePlayer_23, properties {soundWaveAssetPtr=&/Spot_Neon_City_Arcade_Sounds_18}}
 {id=WavePlayer_24, properties {soundWaveAssetPtr=&Spot_Neon_City_Arcade_Sounds_07}}
 {id=WavePlayer_25, properties {soundWaveAssetPtr=&Spot_Neon_City_Arcade_Sounds_08}}
 {id=WavePlayer_26, properties {soundWaveAssetPtr=&Spot_Neon_City_Arcade_Sounds_09}}
 {id=WavePlayer_27, properties {soundWaveAssetPtr=&Spot_Neon_City_Arcade_Sounds_10}}
 {id=WavePlayer_28, properties {soundWaveAssetPtr=&Spot_Neon_City_Arcade_Sounds_11}}
 {id=WavePlayer_29, properties {soundWaveAssetPtr=&Spot_Neon_City_Arcade_Sounds_12}}
 {id=WavePlayer_3, properties {soundWaveAssetPtr=&/Spot_Neon_City_Arcade_Sounds_04}}
 {id=WavePlayer_30, properties {soundWaveAssetPtr=&Spot_Neon_City_Arcade_Sounds_13}}
 {id=WavePlayer_31, properties {soundWaveAssetPtr=&Spot_Neon_City_Arcade_Sounds_14}}
 {id=WavePlayer_32, properties {soundWaveAssetPtr=&Spot_Neon_City_Arcade_Sounds_15}}
 {id=WavePlayer_33, properties {soundWaveAssetPtr=&Spot_Neon_City_Arcade_Sounds_16}}
 {id=WavePlayer_34, properties {soundWaveAssetPtr=&Spot_Neon_City_Arcade_Sounds_17}}
 {id=WavePlayer_35, properties {soundWaveAssetPtr=&Spot_Neon_City_Arcade_Sounds_18}}
 {id=WavePlayer_4, properties {soundWaveAssetPtr=&Spot_Neon_City_Arcade_Sounds_05}}
 {id=WavePlayer_5, properties {soundWaveAssetPtr=&Spot_Neon_City_Arcade_Sounds_06}}
 {id=WavePlayer_6, properties {soundWaveAssetPtr=&Spot_Neon_City_Arcade_Sounds_01}}
 {id=WavePlayer_7, properties {soundWaveAssetPtr=&Spot_Neon_City_Arcade_Sounds_02}}
 {id=WavePlayer_8, properties {soundWaveAssetPtr=&Spot_Neon_City_Arcade_Sounds_03}}
 {id=WavePlayer_9, properties {soundWaveAssetPtr=&Spot_Neon_City_Arcade_Sounds_04}}
]
properties {
 attenuationSettings=&/Spot_Neon_City_Commercial_Att
 bOverrideConcurrency=true
 maxDistance=11500
 soundClassObject=&Fort_Ambient_Athena
 volumeMultiplier=0.752
 concurrencyOverrides {maxCount=3, retriggerTime=3}
}
