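# Looping ambient spot cue that plays rural-variant grind rail stop announcements for the Neon City POI, with attenuation and a concurrency override.
# Each loop waits a random delay, then picks one of two chime-then-voice sequences at 3:1 odds via repeated child entries.
# The common one concatenates a random chime and random voice line; the rarer uses a fixed jingle and line.
# Chimes and jingle are mixed below full volume.
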SoundCue /Game/Sounds/Asteria/POI/NeonCity/Spots/Spot_Neon_City_Rail_Stop_Jingle_Blue_Rural_Cue
root=Looping_1
nodes [
 {id=Concatenator_0, children [Mixer_3,Random_5], properties {inputVolume [1,1]}}
 {id=Concatenator_1, children [Mixer_2,WavePlayer_8], properties {inputVolume [1,1]}}
 {id=Delay_1, children [Random_1], properties {delayMax=10, delayMin=5}}
 {id=Looping_1, children [Delay_1]}
 {id=Mixer_2, children [WavePlayer_7], properties {inputVolume [0.6]}}
 {id=Mixer_3, children [Random_3], properties {inputVolume [0.6]}}
 {id=Random_1, children [Concatenator_0, Concatenator_0, Concatenator_0, Concatenator_1], properties {weights [1,1,1,1]}}
 {
  id=Random_3
  children [WavePlayer_14, WavePlayer_15, WavePlayer_18, WavePlayer_19, WavePlayer_20]
  properties {bShouldExcludeFromBranchCulling=true, weights [1,1,1,1,1]}
 }
 {
  id=Random_5
  children [WavePlayer_21, WavePlayer_22, WavePlayer_23, WavePlayer_32, WavePlayer_30, WavePlayer_31]
  properties {bShouldExcludeFromBranchCulling=true, weights [1,1,1,1,1,1]}
 }
 {id=WavePlayer_14, properties {soundWaveAssetPtr=&/Spot_Grind_Rail_Announcemnt_Chime_01_Rural}}
 {id=WavePlayer_15, properties {soundWaveAssetPtr=&/Spot_Grind_Rail_Announcemnt_Chime_02_Rural}}
 {id=WavePlayer_18, properties {soundWaveAssetPtr=&/Spot_Grind_Rail_Announcemnt_Chime_03_Rural}}
 {id=WavePlayer_19, properties {soundWaveAssetPtr=&/Spot_Grind_Rail_Announcemnt_Chime_04_Rural}}
 {id=WavePlayer_20, properties {soundWaveAssetPtr=&/Spot_Grind_Rail_Announcemnt_Chime_05_Rural}}
 {id=WavePlayer_21, properties {soundWaveAssetPtr=&/newsgrind_vo_space_009_rural}}
 {id=WavePlayer_22, properties {soundWaveAssetPtr=&/newsgrind_vo_space_010_rural}}
 {id=WavePlayer_23, properties {soundWaveAssetPtr=&/newsgrind_vo_space_011_rural}}
 {id=WavePlayer_30, properties {soundWaveAssetPtr=&/newsgrind_vo_space_021_rural}}
 {id=WavePlayer_31, properties {soundWaveAssetPtr=&/newsgrind_vo_space_022_rural}}
 {id=WavePlayer_32, properties {soundWaveAssetPtr=&/newsgrind_vo_space_020_rural}}
 {id=WavePlayer_7, properties {soundWaveAssetPtr=&/Spot_Grind_Rail_Announcemnt_Jingle_07_Rural}}
 {id=WavePlayer_8, properties {soundWaveAssetPtr=&/newsgrind_vo_space_013_rural}}
]
properties {
 attenuationSettings=&/Spot_Neon_City_Rail_Stop_Att
 bOverrideConcurrency=true
 maxDistance=4000
 soundClassObject=&Fort_Ambient_Athena
 volumeMultiplier=1.002
 concurrencyOverrides {maxCount=2, retriggerTime=3}
}
