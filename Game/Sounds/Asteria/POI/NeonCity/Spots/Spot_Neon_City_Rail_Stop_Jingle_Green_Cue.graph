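# Looping spot ambience cue for the Neon City "Green" rail stop, playing announcement jingles and voice lines.
# After a randomized delay, a random node picks between two concatenated sequences, weighted three-to-one by listing the common branch three times.
# The common sequence plays a random chime then a random voice line; the rare one pairs a fixed jingle with a dedicated line.
# Attenuation and a concurrency override constrain range and overlapping instances.
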SoundCue /Game/Sounds/Asteria/POI/NeonCity/Spots/Spot_Neon_City_Rail_Stop_Jingle_Green_Cue
root=Looping_0
nodes [
 {id=Concatenator_2, children [Mixer_1,Random_0], properties {inputVolume [1,1]}}
 {id=Concatenator_3, children [Mixer_0, WavePlayer_13], properties {inputVolume [1,1]}}
 {id=Delay_0, children [Random_4], properties {delayMax=10, delayMin=5}}
 {id=Looping_0, children [Delay_0]}
 {id=Mixer_0, children [WavePlayer_9], properties {inputVolume [0.85]}}
 {id=Mixer_1, children [Random_2], properties {inputVolume [0.85]}}
 {
  id=Random_0
  children [WavePlayer_5, WavePlayer_10, WavePlayer_11, WavePlayer_12, WavePlayer_16, WavePlayer_17]
  properties {bShouldExcludeFromBranchCulling=true, weights [1,1,1,1,1,1]}
 }
 {
  id=Random_2
  children [WavePlayer_0,WavePlayer_1,WavePlayer_2,WavePlayer_3,WavePlayer_4]
  properties {bShouldExcludeFromBranchCulling=true, weights [1,1,1,1,1]}
 }
 {id=Random_4, children [Concatenator_2, Concatenator_2, Concatenator_2, Concatenator_3], properties {weights [1,1,1,1]}}
 {id=WavePlayer_0, properties {soundWaveAssetPtr=&/Spot_Grind_Rail_Announcemnt_Chime_01}}
 {id=WavePlayer_1, properties {soundWaveAssetPtr=&/Spot_Grind_Rail_Announcemnt_Chime_02}}
 {id=WavePlayer_10, properties {soundWaveAssetPtr=&/newsgrind_vo_aoike_010}}
 {id=WavePlayer_11, properties {soundWaveAssetPtr=&/newsgrind_vo_aoike_011}}
 {id=WavePlayer_12, properties {soundWaveAssetPtr=&/newsgrind_vo_aoike_020}}
 {id=WavePlayer_13, properties {soundWaveAssetPtr=&/newsgrind_vo_aoike_012}}
 {id=WavePlayer_16, properties {soundWaveAssetPtr=&/newsgrind_vo_aoike_021}}
 {id=WavePlayer_17, properties {soundWaveAssetPtr=&/newsgrind_vo_aoike_022}}
 {id=WavePlayer_2, properties {soundWaveAssetPtr=&/Spot_Grind_Rail_Announcemnt_Chime_03}}
 {id=WavePlayer_3, properties {soundWaveAssetPtr=&/Spot_Grind_Rail_Announcemnt_Chime_04}}
 {id=WavePlayer_4, properties {soundWaveAssetPtr=&/Spot_Grind_Rail_Announcemnt_Chime_05}}
 {id=WavePlayer_5, properties {soundWaveAssetPtr=&/newsgrind_vo_aoike_009}}
 {id=WavePlayer_9, properties {soundWaveAssetPtr=&/Spot_Grind_Rail_Announcemnt_Jingle_08}}
]
properties {
 attenuationSettings=&/Spot_Neon_City_Rail_Stop_Att
 bOverrideConcurrency=true
 maxDistance=4000
 soundClassObject=&Fort_Ambient_Athena
 volumeMultiplier=0.669
 concurrencyOverrides {maxCount=2, retriggerTime=3}
}
